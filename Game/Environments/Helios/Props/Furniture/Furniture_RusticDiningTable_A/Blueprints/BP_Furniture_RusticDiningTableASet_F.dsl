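; Blueprint child actor for Set F of the Rustic Dining Table A furniture prop in the Helios environment.
; Its construction script only calls the parent's, so any setup logic is inherited.
; When the actor is destroyed, it also destroys its Box1 component.
; BeginPlay, Tick and ActorBeginOverlap are empty stubs.
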(event EventBeginPlay)

(event Collision|EventActorBeginOverlap
  (params (OtherActor :type Actor%20Object%20Reference)))

(event EventTick
  (params (DeltaSeconds :type Float%20%28single-precision%29)))

(event EventDestroyed
  (Components|DestroyComponent (Variables|Default|GetBox1)))

(fn ConstructionScript
  (|CallParentFunction))
